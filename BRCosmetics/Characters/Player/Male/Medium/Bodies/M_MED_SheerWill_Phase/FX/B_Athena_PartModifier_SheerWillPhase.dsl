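; Character part modifier for the Sheer Will Phase body cosmetic that adjusts its idle VFX on weapon change.
; After the parent call and a short retriggerable delay, where cosmetic events can execute, it checks the new weapon's mesh against the Sheer Will pickaxe.
; On a match it prints a debug string and sets the idle VFX's bisHandHidden parameter to false.
; Other weapons leave the parameter untouched here.
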(event EventOnWeaponChanged
  (params (NewWeapon :type Fort%20Weapon%20Object%20Reference) (OldWeapon :type Fort%20Weapon%20Object%20Reference))
  (|CallParentFunction NewWeapon OldWeapon)
  (Utilities|FlowControl|RetriggerableDelay 0.2)
  (Networking|CanExecuteCosmeticEvents
    (:True
      (if (== (Components|SkeletalMesh|GetSkeletalMeshAsset (Variables|/Script/FortniteGame.FortWeapon|GetWeaponMesh NewWeapon)) "/BRCosmetics/Weapons/FORT_Melee/Pickaxe_SheerWill/Meshes/SheerWill_Axe.SheerWill_Axe")
        (Development|PrintString "True" true true "(R=0.909465,G=0.015480,B=1.000000,A=1.000000)" 2e+01)
        (Effects|Components|ParticleSystem|SetBoolParameter (CallFunction|GetIdleVFXComponent) "bisHandHidden" false)))))

(fn ConstructionScript
  (|CallParentFunction))
