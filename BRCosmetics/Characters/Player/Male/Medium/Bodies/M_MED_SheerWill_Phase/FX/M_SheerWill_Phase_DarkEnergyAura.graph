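# Translucent unlit Niagara particle material for the dark energy aura on the SheerWill Phase character cosmetic.
# Opacity blends scrolling trail, fire and wisp textures, modulated by cell noise, particle alpha, a noise-distorted radial gradient and depth fade.
# Emissive mixes dark purples with a pink highlight through panning noise, tinted by particle color.
# Per-particle random values offset and scale the UVs so particles differ.
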Material /BRCosmetics/Characters/Player/Male/Medium/Bodies/M_MED_SheerWill_Phase/FX/M_SheerWill_Phase_DarkEnergyAura
graph {
 properties {
  bUsedWithNiagaraMeshParticles=true
  bUsedWithNiagaraSprites=true
  blendMode=BLEND_Translucent
  pixelDepthOffsetMode=PDOM_Legacy
  shadingModel=MSM_Unlit
  usedShadingModels=Unlit
 }
 nodes [
  n1 = ParticleColor()
  n2 = TextureCoordinate() {uTiling=1.2, vTiling=0.5}
  n3 = NamedRerouteUsage()
  n4 = Add(A=Add(A=Multiply(A=@n2, B=@n3)) {constB=0}, B=AppendVector(A=Constant() {r=0.15}, B=ParticleRandom()))
  n5 = TextureSample(UVs=MaterialFunctionCall(XY=Panner(Coordinate=@n4) {speedY=0.35}) {materialFunction=&Swizzle}.YX) {samplerSource=SSM_Wrap_WorldGroupSettings, texture=/BRCosmetics/Weapons/FORT_Melee/Pickaxe_SoilBlend/FX/T_Pickaxe_SoilBlend_TrailMask}
  n6 = TextureCoordinate() {vTiling=0.5}
  n7 = NamedRerouteUsage()
  n8 = Add(A=Add(A=Multiply(A=@n6, B=@n7)) {constB=0}, B=AppendVector(A=Constant(), B=ParticleRandom()))
  n9 = TextureSample(UVs=MaterialFunctionCall(XY=Panner(Coordinate=@n8) {speedY=0.35}) {materialFunction=&Swizzle}.YX) {
   samplerSource=SSM_Wrap_WorldGroupSettings
   samplerType=SAMPLERTYPE_LinearGrayscale
   texture=/BRCosmetics/Athena/Effects/Particles/Skydiving/NS_Contrail_LightCycle/Textures/T_FireTrail_01
  }
  n10 = TextureCoordinate() {vTiling=0.5}
  n11 = NamedRerouteUsage()
  n12 = Add(A=Add(A=Multiply(A=@n10, B=@n11)) {constB=0}, B=AppendVector(A=Constant(), B=ParticleRandom()))
  n13 = TextureSample(UVs=MaterialFunctionCall(XY=Panner(Coordinate=@n12) {speedY=0.35}) {materialFunction=&Swizzle}.YX) {
   samplerSource=SSM_Wrap_WorldGroupSettings
   texture=/BRCosmetics/Athena/Effects/Particles/Skydiving/NS_Contrail_HomeRange/T_Contrail_HomeRange_WispyTrail
  }
  n14 = LinearInterpolate(A=Power(Base=@n9.R) {constExponent=1.2}, B=@n13.R)
  n15 = Add(A=Multiply(A=TextureCoordinate() {uTiling=1.3}, B=NamedRerouteUsage()), B=AppendVector(A=Constant(), B=ParticleRandom()))
  n16 = LinearInterpolate(Alpha=TextureSample(UVs=Panner(Coordinate=@n15) {speedY=0.5}) {samplerType=SAMPLERTYPE_Grayscale, texture=/BRCosmetics/Weapons/FORT_Melee/Pickaxe_Cube_Ninja_Male/FX/T_CellNoise}.R) {constA=0.4}
  n17 = Multiply(A=Power(Base=LinearInterpolate(A=@n5.R, B=@n14) {constAlpha=0.35}) {constExponent=1}, B=@n16)
  n18 = Saturate(None=Multiply(A=Multiply(A=@n1.A, B=@n17)) {constB=2})
  n19 = Constant2Vector() {g=1}
  n20 = Add(A=TextureCoordinate() {uTiling=1.2, vTiling=0.7}, B=AppendVector(A=Constant(), B=ParticleRandom()))
  n21 = Multiply(
   A=TextureSample(UVs=Panner(Coordinate=@n20) {speedY=0.6}) {samplerType=SAMPLERTYPE_Grayscale, texture=&LowResBlurredNoise}.R
   B=ScalarParameter() {defaultValue=0.2, parameterName=Noise_Intensity}
  ) {constB=0.2}
  n22 = Add(
   A=Add(A=TextureCoordinate(), B=Constant2Vector() {g=-0.07})
   B=NamedRerouteDeclaration(None=Multiply(A=@n19, B=@n21)) {name=Disturb, nodeColor=<rgba 0 0.0854 1 1>}
  )
  n23 = ScalarParameter() {defaultValue=0.45, parameterName="Radius (S)"}
  n24 = ScalarParameter() {defaultValue=3, parameterName="Density (S)"}
  DepthFade(Opacity=Multiply(
    A=Multiply(A=@n18)
    B=MaterialFunctionCall(UVs=@n22, Radius=@n23, Density=@n24) {materialFunction=&RadialGradientExponential}.RadialGradientExponential
   )) {fadeDistanceDefault=5}
  n25 = Panner(Coordinate=TextureCoordinate() {uTiling=1.2}) {speedY=0.5}
  n26 = LinearInterpolate(
   A=VectorParameter() {parameterName=DarkColor, defaultValue=<rgba 0.0045 0 0.02 1>}.RGB
   B=Constant4Vector() {constant=<rgba 0.025 0 0.15 0>}.RGB
   Alpha=Power(Base=TextureSample(UVs=@n25) {samplerType=SAMPLERTYPE_Grayscale, texture=&T_CellNoise}.R) {constExponent=1.5}
  )
  n27 = VectorParameter() {parameterName=HighLightColor, defaultValue=<rgba 1 0 0.3333 1>}
  n28 = TextureCoordinate() {uTiling=1.2}
  n29 = LinearInterpolate(Alpha=ParticleRandom()) {constA=0.9}
  n30 = Panner(Coordinate=Add(A=Multiply(A=@n28, B=@n29), B=ParticleRandom())) {speedY=0.4}
  n31 = Multiply(
   A=Power(Base=TextureSample(UVs=@n30) {texture=/Game/Weapons/FORT_Melee/Pickaxe_CoatCheck/FX/T_Pickaxe_CoatCheck_Noise}.R) {constExponent=1.2}
   B=@n16
  )
  MaterialFunctionCall(Emissive=Multiply(A=@n1.RGB, B=LinearInterpolate(A=@n26, B=@n27.RGB, Alpha=@n31))) {materialFunction=&MF_FX_EmissiveAdjustments}
  n32 = ParticleRandom()
  NamedRerouteDeclaration(None=AppendVector(A=Constant() {r=1}, B=LinearInterpolate(Alpha=@n32) {constA=0.8})) {name=RandomUVScale, nodeColor=<rgba 1 0.065 0 1>}
 ]
}
