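# Translucent unlit Niagara sprite material for the SheerWill Phase character FX, producing screen-space distortion around the body.
# Two panning noise heightmaps become blended normals that offset SceneColor UVs, scaled by DistortionIntensity and the BodyDisturbIntensity particle parameter.
# A SubsurfaceColor threshold masks the character, subtracting the product of four offset neighbor samples from the distorted center sample to isolate edges.
# Shading path switches return constant zero on Forward and Mobile, dropping both effects there.
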Material /BRCosmetics/Characters/Player/Male/Medium/Bodies/M_MED_SheerWill_Phase/FX/M_SheerWill_Phase_SceneTextureEffect
graph {
 properties {
  bAutomaticallySetUsageInEditor=false
  bUsedWithNiagaraSprites=true
  blendMode=BLEND_Translucent
  shadingModel=MSM_Unlit
  usedShadingModels=Unlit
 }
 nodes [
  Constant()
  LinearInterpolate(Alpha=MaterialFunctionCall(Distance=Constant() {r=300}) {materialFunction=&Near_Camera_Fade}.Result) {constA=2, constB=6}
  ScreenPosition()
  n1 = Constant()
  n2 = ParticlePositionWS()
  n3 = Multiply(A=Multiply(B=MaterialFunctionCall!(Center Point World Position=@n2) {materialFunction=/Engine/Functions/Engine_MaterialFunctions02/UVs/WorldSpaceAlignedScreenCoordinates}.Offset and Scaled UVS) {constA=10}) {constB=4}
  n4 = Multiply(
   A=Constant2Vector() {g=2, r=2}
   B=ComponentMask(None=VectorParameter() {parameterName=U_Scale, defaultValue=<rgba 0.8 1 0 0>}.RGB) {g=true, r=true}
  )
  n5 = Multiply(A=Time(), B=ScalarParameter() {defaultValue=1.3, parameterName=Speed})
  n6 = MaterialFunctionCall!(
   Height Map=TextureObject() {samplerType=SAMPLERTYPE_LinearGrayscale, texture=/Game/Effects/Fort_Effects/Emotes/UltraLight/Texture/T_DirectionalNoise_Greyscale}
   Normal Map Intensity
   Height Map UV Offset
   Coordinates=Panner(Coordinate=Multiply(A=@n3, B=@n4), Time=@n5) {speedY=0.15}
   Height Map Channel Selector
  ) {materialFunction=/Engine/Functions/Engine_MaterialFunctions03/Procedurals/NormalFromHeightmap}
  n7 = TextureObject() {samplerType=SAMPLERTYPE_Grayscale, texture=/BRCosmetics/Effects/Fort_Effects/Emotes/SwissKale/T_Blur_Noise}
  n8 = Constant2Vector() {g=5, r=10}
  n9 = ComponentMask(None=VectorParameter() {parameterName=U_Scale, defaultValue=<rgba 0.8 1 0 0>}.RGB) {g=true, r=true}
  n10 = Panner(Coordinate=Multiply(A=@n3, B=Multiply(A=@n8, B=@n9)), Time=@n5) {speedY=0.3}
  n11 = LinearInterpolate(
   A=MaterialFunctionCall(Normal=Multiply(A=@n6.Result), Flatness=ScalarParameter() {defaultValue=0.5, parameterName=Flatness}) {materialFunction=&FlattenNormal}.Result
   B=Multiply(A=MaterialFunctionCall(Height Map=@n7, Coordinates=@n10) {materialFunction=&NormalFromHeightmap}.Result)
   Alpha=ScalarParameter() {defaultValue=0.6, parameterName=NoiseLerp}
  )
  n12 = ScalarParameter() {defaultValue=0.03, parameterName=DistortionIntensity}
  n13 = DynamicParameter() {paramNames [BodyDisturbIntensity, Param2, Param3, Param4]}
  n14 = Add(A=ScreenPosition().ViewportUV, B=Multiply(A=ComponentMask(None=@n11) {g=true, r=true}, B=Multiply(A=@n12, B=@n13.BodyDisturbIntensity))) {constB=0}
  n15 = SceneColor!(UVs=@n14)
  ShadingPathSwitch(Default=@n15.RGB, Deferred=@n15.RGB, Forward=@n1, Mobile=@n1)
  n16 = Constant2Vector() {r=1}
  n17 = ScalarParameter() {defaultValue=0.015, parameterName=Param_2}
  n18 = SceneTexture(UVs=Add(A=ScreenPosition().ViewportUV, B=Multiply(A=@n16, B=@n17))) {sceneTextureId=PPI_SubsurfaceColor}
  n19 = Constant()
  n20 = If(
   A=Min(A=ComponentMask(None=@n18.Color) {g=true}, B=ComponentMask(None=@n18.Color) {b=true})
   B=SquareRoot(None=ScalarParameter() {defaultValue=0.003, group=Cartoon, parameterName=Subsurface Mask})
   "A > B"=@n19
   "A == B"=Constant() {r=1}
   "A < B"=@n19
  ) {equalsThreshold=0.0005}
  n21 = Constant2Vector() {r=-1}
  n22 = ScalarParameter() {defaultValue=0.015, parameterName=Param_2}
  n23 = SceneTexture(UVs=Add(A=ScreenPosition().ViewportUV, B=Multiply(A=@n21, B=@n22))) {sceneTextureId=PPI_SubsurfaceColor}
  n24 = Constant()
  n25 = If(
   A=Min(A=ComponentMask(None=@n23.Color) {g=true}, B=ComponentMask(None=@n23.Color) {b=true})
   B=SquareRoot(None=ScalarParameter() {defaultValue=0.003, group=Cartoon, parameterName=Subsurface Mask})
   "A > B"=@n24
   "A == B"=Constant() {r=1}
   "A < B"=@n24
  ) {equalsThreshold=0.0005}
  n26 = Constant2Vector() {g=1}
  n27 = ScalarParameter() {defaultValue=0.015, parameterName=Param_2}
  n28 = SceneTexture(UVs=Add(A=ScreenPosition().ViewportUV, B=Multiply(A=@n26, B=@n27))) {sceneTextureId=PPI_SubsurfaceColor}
  n29 = Constant()
  n30 = If(
   A=Min(A=ComponentMask(None=@n28.Color) {g=true}, B=ComponentMask(None=@n28.Color) {b=true})
   B=SquareRoot(None=ScalarParameter() {defaultValue=0.003, group=Cartoon, parameterName=Subsurface Mask})
   "A > B"=@n29
   "A == B"=Constant() {r=1}
   "A < B"=@n29
  ) {equalsThreshold=0.0005}
  n31 = Constant2Vector() {g=-1}
  n32 = ScalarParameter() {defaultValue=0.015, parameterName=Param_2}
  n33 = SceneTexture(UVs=Add(A=ScreenPosition().ViewportUV, B=Multiply(A=@n31, B=@n32))) {sceneTextureId=PPI_SubsurfaceColor}
  n34 = Constant()
  n35 = If(
   A=Min(A=ComponentMask(None=@n33.Color) {g=true}, B=ComponentMask(None=@n33.Color) {b=true})
   B=SquareRoot(None=ScalarParameter() {defaultValue=0.003, group=Cartoon, parameterName=Subsurface Mask})
   "A > B"=@n34
   "A == B"=Constant() {r=1}
   "A < B"=@n34
  ) {equalsThreshold=0.0005}
  n36 = SceneTexture(UVs=@n14) {sceneTextureId=PPI_SubsurfaceColor}
  n37 = ComponentMask(None=@n36.Color) {g=true}
  n38 = ComponentMask(None=@n36.Color) {b=true}
  n39 = ScalarParameter() {defaultValue=0.003, group=Cartoon, parameterName=Subsurface Mask}
  n40 = Constant()
  n41 = Subtract(
   A=If(A=Min(A=@n37, B=@n38), B=SquareRoot(None=@n39), "A > B"=@n40, "A == B"=Constant() {r=1}, "A < B"=@n40) {equalsThreshold=0.0005}
   B=Multiply(A=Multiply(A=@n20, B=@n25), B=Multiply(A=@n30, B=@n35))
  )
  n42 = Constant()
  ShadingPathSwitch(Default=@n41, Deferred=@n41, Forward=@n42, Mobile=@n42)
 ]
}
