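# Material function for the SheerWill Phase character body that outputs an animated, colored Fresnel mask plus supporting noise and UV outputs.
# A triplanar UV helper borrowed from BirchPalace supplies front-projection UVs, which are also output directly and drive panning noise samples.
# A sharpened vertex-normal Fresnel multiplies inverse noise and a dark-to-HDR color blend keyed by soft-cell noise.
# Raw cell noise feeds a WPO output, and separately panned blurred noise feeds NoiseA.
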MaterialFunction /BRCosmetics/Characters/Player/Male/Medium/Bodies/M_MED_SheerWill_Phase/Material/MF_SheerWill_FresnelMask
graph {nodes [
  n1 = MaterialFunctionCall!(
   TriplanarUV_UseClamp?=Constant()
   TriplanarUV_FlipUV=Constant()
   TriplanarUV_CenterPosition=Constant4Vector() {constant=<rgba 0.9523 0 15.2308 1>}.RGB
   TriplanarUV_UVScale=Constant() {r=58.5987}
  ) {materialFunction=/BRCosmetics/Characters/Player/Male/Medium/Bodies/M_MED_BirchPalace/Skins/Gas/Materials/MF_TriplanarUV}
  FunctionOutput(None=@n1.UV_FrontProjection) {outputName=ProjectionForntUV}
  n2 = Power(Base=Fresnel(ExponentIn=Constant() {r=1.1518}, Normal=VertexNormalWS()), Exp=Constant() {r=6.0922}) {constExponent=3}
  n3 = Panner(Coordinate=Multiply(A=@n1.UV_FrontProjection, B=ScalarParameter() {defaultValue=1, parameterName=FresnelMask_tiling})) {speedX=0.04, speedY=0.04}
  n4 = TextureSample(UVs=@n3) {
   samplerSource=SSM_Wrap_WorldGroupSettings
   samplerType=SAMPLERTYPE_LinearColor
   texture=/Game/Global/Textures/Default/Noise/T_Noise01_Inverse
  }
  n5 = Constant4Vector() {constant=<rgba 0.0298 0 0.1 1>}
  n6 = Constant4Vector() {constant=<rgba 10 0 2.4079 1>}
  n7 = TextureSample(UVs=@n3) {samplerSource=SSM_Wrap_WorldGroupSettings, texture=/Game/Global/Textures/Default/Noise/T_SoftCells_Negative}
  n8 = Saturate(None=Power(Base=@n7.R))
  FunctionOutput(None=Multiply(A=Multiply(A=@n2, B=@n4.R), B=LinearInterpolate(A=@n5.RGB, B=@n6.RGB, Alpha=@n8))) {outputName=SheerWill_FresnelMask}
  FunctionOutput(None=@n7.R) {outputName=SheerWill_FresnelMask_WPO}
  n9 = Constant() {r=1}
  n10 = TextureSample(UVs=Panner(Coordinate=Multiply(A=@n1.UV_FrontProjection, B=@n9)) {speedY=0.1}) {
   samplerSource=SSM_Wrap_WorldGroupSettings
   samplerType=SAMPLERTYPE_Grayscale
   texture=/Game/Global/Textures/Default/Noise/LowResBlurredNoiseNormalized
  }
  FunctionOutput(None=@n10.R) {outputName=NoiseA}
 ]}
