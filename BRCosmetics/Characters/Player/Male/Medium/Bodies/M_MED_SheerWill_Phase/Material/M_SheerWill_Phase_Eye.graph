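# Character eye material for the SheerWill Phase body cosmetic, built on material attributes for skeletal meshes with a shading model set by expression.
# The shared 2023 character parent PreFX function feeds the PostFX function, with no custom effect between them.
# A side branch projects a soft-cell noise texture over scaled pre-skinned local XZ position and sharpens it with a power into EdgeNoise2.
# No visible node clearly consumes that noise.
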Material /BRCosmetics/Characters/Player/Male/Medium/Bodies/M_MED_SheerWill_Phase/Material/M_SheerWill_Phase_Eye
graph {
 properties {
  bAutomaticallySetUsageInEditor=false
  bUseMaterialAttributes=true
  bUsedWithMeshDeformer=true
  bUsedWithMorphTargets=true
  bUsedWithSkeletalMesh=true
  pixelDepthOffsetMode=PDOM_Legacy
  shadingModel=MSM_FromMaterialExpression
  usedShadingModels=Default Lit | Subsurface | Clear Coat | Subsurface Profile | Cloth
 }
 nodes [
  n1 = MaterialFunctionCall() {materialFunction=&MF_CharacterParent_2023_PreFX}
  MaterialFunctionCall(Material=@n1.Result, M=@n1.Result, Pre Skinned Local Position=@n1.Result, Distance Blend=@n1.Result, Fresnel=@n1.Result) {materialFunction=&MF_CharacterParent_2023_PostFX}
  n2 = Multiply(
   A=ComponentMask(None=LocalPosition() {localOrigin=InstancePreSkinning}.XYZ) {b=true, r=true}
   B=ScalarParameter() {defaultValue=0.08, parameterName=Param_14}
  )
  NamedRerouteDeclaration(None=Power(
    Base=TextureSample(UVs=@n2) {texture=/Game/Global/Textures/Default/Noise/T_SoftCells_Negative}.R
    Exp=ScalarParameter() {defaultValue=1.5, parameterName=Param_15}
   )) {name=EdgeNoise2, nodeColor=<rgba 1 0.1172 0 1>}
  VertexInterpolator(VS=NamedRerouteUsage())
 ]
}
