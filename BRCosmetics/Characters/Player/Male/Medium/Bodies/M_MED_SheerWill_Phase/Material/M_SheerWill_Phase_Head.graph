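# Character cloth-shaded material for the head of the SheerWill Phase outfit, with near-black base color, animated emissive glow and vertex wobble.
# Emissive layers combine tinted FX-mask lines pulsed by panning noise and shaped by AO and face mask, a noise-driven Fresnel rim and gradient glow.
# Panning local-position noise drives normal-direction world position offset, suppressed on camera-facing, emissive and hand-masked areas.
# Shared CharacterParent PreFX and PostFX functions wrap the overrides.
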Material /BRCosmetics/Characters/Player/Male/Medium/Bodies/M_MED_SheerWill_Phase/Material/M_SheerWill_Phase_Head
graph {
 expressions {
  MaterialExpressionAdd_0 {
   class=MaterialExpressionAdd
   inputs [
    A {from=MaterialExpressionNamedRerouteUsage_19}
    B {from=MaterialExpressionAdd_4}
   ]
  }
  MaterialExpressionAdd_10 {
   class=MaterialExpressionAdd
   inputs [
    A {from=MaterialExpressionTextureCoordinate_3}
    B {from=MaterialExpressionAppendVector_1}
   ]
  }
  MaterialExpressionAdd_11 {
   class=MaterialExpressionAdd
   inputs [
    A {from=MaterialExpressionMax_0}
    B {from=MaterialExpressionMultiply_49}
   ]
  }
  MaterialExpressionAdd_12 {
   class=MaterialExpressionAdd
   inputs [
    A {from=MaterialExpressionNamedRerouteUsage_18}
    B {from=MaterialExpressionAdd_0}
   ]
  }
  MaterialExpressionAdd_13 {
   class=MaterialExpressionAdd
   inputs [
    A {from=MaterialExpressionMax_7}
    B {from=MaterialExpressionPower_8}
   ]
  }
  MaterialExpressionAdd_4 {
   class=MaterialExpressionAdd
   inputs [
    A {from=MaterialExpressionAdd_5}
    B {from=MaterialExpressionConstant4Vector_10, output=RGB}
   ]
  }
  MaterialExpressionAdd_5 {
   class=MaterialExpressionAdd
   inputs [
    A {from=MaterialExpressionNamedRerouteUsage_15}
    B {from=MaterialExpressionLinearInterpolate_17}
   ]
  }
  MaterialExpressionAppendVector_1 {
   class=MaterialExpressionAppendVector
   inputs [
    A {from=MaterialExpressionConstant_11, output=""}
    B {from=MaterialExpressionMultiply_48, output=""}
   ]
  }
  MaterialExpressionCameraPositionWS_1 {class=MaterialExpressionCameraPositionWS, inputs [], properties {}}
  MaterialExpressionComponentMask_0 {class=MaterialExpressionComponentMask, inputs [None {from=MaterialExpressionLocalPosition_1, output=XYZ}], properties {b=true, r=true}}
  MaterialExpressionComponentMask_2 {class=MaterialExpressionComponentMask, inputs [None {from=MaterialExpressionTextureCoordinate_4}], properties {g=true}}
  MaterialExpressionConstant4Vector_10 {class=MaterialExpressionConstant4Vector, inputs [], properties {constant=<rgba 0.0025 0 0.004 1>}}
  MaterialExpressionConstant4Vector_7 {class=MaterialExpressionConstant4Vector, inputs [], properties {constant=<rgba 0.0349 0 0.25 1>}}
  MaterialExpressionConstant4Vector_8 {class=MaterialExpressionConstant4Vector, inputs [], properties {constant=<rgba 0 0.0168 0.2 1>}}
  MaterialExpressionConstant4Vector_9 {class=MaterialExpressionConstant4Vector, inputs [], properties {constant=<rgba 0.0168 0 0.05 1>}}
  MaterialExpressionConstant_1 {class=MaterialExpressionConstant, properties {r=0.001}}
  MaterialExpressionConstant_10 {class=MaterialExpressionConstant, properties {r=0.02}}
  MaterialExpressionConstant_11 {class=MaterialExpressionConstant}
  MaterialExpressionConstant_12 {class=MaterialExpressionConstant, properties {r=1.5}}
  MaterialExpressionConstant_14 {class=MaterialExpressionConstant, properties {r=0.08}}
  MaterialExpressionConstant_16 {class=MaterialExpressionConstant, properties {r=12}}
  MaterialExpressionConstant_27 {class=MaterialExpressionConstant, properties {r=-1}}
  MaterialExpressionConstant_30 {class=MaterialExpressionConstant, properties {r=0.121}}
  MaterialExpressionConstant_31 {class=MaterialExpressionConstant, properties {r=0.0934}}
  MaterialExpressionConstant_32 {class=MaterialExpressionConstant, properties {r=7}}
  MaterialExpressionConstant_33 {class=MaterialExpressionConstant, properties {r=6.2509}}
  MaterialExpressionConstant_34 {class=MaterialExpressionConstant, properties {r=4}}
  MaterialExpressionConstant_35 {class=MaterialExpressionConstant, properties {r=0.3}}
  MaterialExpressionConstant_36 {class=MaterialExpressionConstant, properties {r=0.004}}
  MaterialExpressionConstant_7 {class=MaterialExpressionConstant, properties {r=0.5}}
  MaterialExpressionConvert_1 {
   class=MaterialExpressionConvert
   inputs [
    None {from=""}
    None {from=MaterialExpressionScalarParameter_5}
    None {from=MaterialExpressionScalarParameter_5}
   ]
   properties {
    nodeName=Make Vector3
    convertInputs [
     {type=Scalar, expressionInput {expression=null, inputName=None, mask=0, maskA=0, maskB=0, maskG=0, maskR=0, outputIndex=0}}
     {
      type=Scalar
      expressionInput {
       expression=&self:MaterialExpressionScalarParameter_5
       inputName=None
       mask=0
       maskA=0
       maskB=0
       maskG=0
       maskR=0
       outputIndex=0
      }
     }
     {
      type=Scalar
      expressionInput {
       expression=&self:MaterialExpressionScalarParameter_5
       inputName=None
       mask=0
       maskA=0
       maskB=0
       maskG=0
       maskR=0
       outputIndex=0
      }
     }
    ]
    convertOutputs [{type=Vector3}]
   }
  }
  MaterialExpressionDotProduct_1 {
   class=MaterialExpressionDotProduct
   inputs [
    A {from=MaterialExpressionSubtract_1, output=""}
    B {from=MaterialExpressionVertexNormalWS_3, output=""}
   ]
   properties {}
  }
  MaterialExpressionFresnel_1 {
   class=MaterialExpressionFresnel
   inputs [
    ExponentIn {from=MaterialExpressionConstant_30, output=""}
    BaseReflectFractionIn {from="", output=""}
    Normal {from=MaterialExpressionVertexNormalWS_0, output=""}
   ]
   properties {}
  }
  MaterialExpressionFresnel_4 {
   class=MaterialExpressionFresnel
   inputs [
    ExponentIn {from=MaterialExpressionConstant_31, output=""}
    BaseReflectFractionIn {from="", output=""}
    Normal {from=MaterialExpressionVertexNormalWS_1, output=""}
   ]
   properties {}
  }
  MaterialExpressionLinearInterpolate_12 {
   class=MaterialExpressionLinearInterpolate
   inputs [
    A {from=""}
    B {from=""}
    Alpha {from=MaterialExpressionTextureSample_10, output=R}
   ]
   properties {constA=1, constB=40}
  }
  MaterialExpressionLinearInterpolate_16 {
   class=MaterialExpressionLinearInterpolate
   inputs [
    A {from=""}
    B {from=MaterialExpressionLinearInterpolate_21}
    Alpha {from=MaterialExpressionTextureSample_11, output=R}
   ]
  }
  MaterialExpressionLinearInterpolate_17 {
   class=MaterialExpressionLinearInterpolate
   inputs [
    A {from=MaterialExpressionLinearInterpolate_16}
    B {from=MaterialExpressionMultiply_34}
    Alpha {from=MaterialExpressionFresnel_1}
   ]
   properties {constB=0}
  }
  MaterialExpressionLinearInterpolate_18 {
   class=MaterialExpressionLinearInterpolate
   inputs [
    A {from=MaterialExpressionConstant_27}
    B {from=""}
    Alpha {from=MaterialExpressionTextureSample_13, output=R}
   ]
   properties {constA=-1}
  }
  MaterialExpressionLinearInterpolate_19 {
   class=MaterialExpressionLinearInterpolate
   inputs [
    A {from=MaterialExpressionScalarParameter_4}
    B {from=MaterialExpressionScalarParameter_6}
    Alpha {from=MaterialExpressionPower_2}
   ]
  }
  MaterialExpressionLinearInterpolate_20 {
   class=MaterialExpressionLinearInterpolate
   inputs [
    A {from=MaterialExpressionConstant_35}
    B {from=MaterialExpressionConstant_34}
    Alpha {from=MaterialExpressionPower_3}
   ]
  }
  MaterialExpressionLinearInterpolate_21 {
   class=MaterialExpressionLinearInterpolate
   inputs [
    A {from=MaterialExpressionConstant4Vector_7, output=RGB}
    B {from=MaterialExpressionConstant4Vector_8, output=RGB}
    Alpha {from=MaterialExpressionTextureSample_17, output=R}
   ]
  }
  MaterialExpressionLinearInterpolate_22 {
   class=MaterialExpressionLinearInterpolate
   inputs [
    A {from=""}
    B {from=""}
    Alpha {from=MaterialExpressionTextureSample_16, output=R}
   ]
   properties {constA=0.35, constB=2}
  }
  MaterialExpressionLinearInterpolate_23 {
   class=MaterialExpressionLinearInterpolate
   inputs [
    A {from=""}
    B {from=""}
    Alpha {from=MaterialExpressionTextureSample_18, output=R}
   ]
   properties {constB=0.4}
  }
  MaterialExpressionLinearInterpolate_24 {
   class=MaterialExpressionLinearInterpolate
   inputs [
    A {from=""}
    B {from=""}
    Alpha {from=MaterialExpressionTextureSample_19, output=R}
   ]
   properties {constB=2}
  }
  MaterialExpressionLinearInterpolate_25 {
   class=MaterialExpressionLinearInterpolate
   inputs [
    A {from=MaterialExpressionVectorParameter_6, output=RGB}
    B {from=MaterialExpressionLinearInterpolate_27}
    Alpha {from=MaterialExpressionNamedRerouteUsage_16}
   ]
  }
  MaterialExpressionLinearInterpolate_26 {
   class=MaterialExpressionLinearInterpolate
   inputs [
    A {from=MaterialExpressionMultiply_53}
    B {from=MaterialExpressionTextureSampleParameter2D_11, output=R}
    Alpha {from=MaterialExpressionSmoothStep_1}
   ]
  }
  MaterialExpressionLinearInterpolate_27 {
   class=MaterialExpressionLinearInterpolate
   inputs [
    A {from=MaterialExpressionVectorParameter_8, output=RGB}
    B {from=MaterialExpressionVectorParameter_7, output=RGB}
    Alpha {from=MaterialExpressionSmoothStep_1}
   ]
  }
  MaterialExpressionLinearInterpolate_28 {
   class=MaterialExpressionLinearInterpolate
   inputs [
    A {from=""}
    B {from=""}
    Alpha {from=MaterialExpressionMax_8}
   ]
   properties {constA=1, constB=0}
  }
  MaterialExpressionLocalPosition_1 {class=MaterialExpressionLocalPosition, inputs [], properties {localOrigin=InstancePreSkinning}}
  MaterialExpressionMaterialFunctionCall_2 {
   class=MaterialExpressionMaterialFunctionCall
   inputs [
    Material {from=MaterialExpressionSetMaterialAttributes_1}
    M {from=MaterialExpressionMaterialFunctionCall_3, output=M}
    Pre Skinned Local Position {from=MaterialExpressionMaterialFunctionCall_3, output=M}
    Distance Blend {from=MaterialExpressionMaterialFunctionCall_3, output=M}
    Fresnel {from=MaterialExpressionMaterialFunctionCall_3, output=M}
   ]
   properties {materialFunction=&MF_CharacterParent_2023_PostFX}
  }
  MaterialExpressionMaterialFunctionCall_3 {
   class=MaterialExpressionMaterialFunctionCall
   inputs [
    Overide Texture Set UVs {from=""}
    Override Tech Art Mask UVs {from=""}
    Override Basic Emissive UVs {from=""}
   ]
   properties {materialFunction=&MF_CharacterParent_2023_PreFX}
  }
  MaterialExpressionMaterialFunctionCall_6 {
   class=MaterialExpressionMaterialFunctionCall
   inputs []
   properties {materialFunction=&/MF_SheerWill_FresnelMask}
  }
  MaterialExpressionMax_0 {
   class=MaterialExpressionMax
   inputs [
    A {from=MaterialExpressionMax_2}
    B {from=MaterialExpressionMultiply_44}
   ]
  }
  MaterialExpressionMax_2 {
   class=MaterialExpressionMax
   inputs [
    A {from=MaterialExpressionMultiply_42}
    B {from=MaterialExpressionMultiply_46}
   ]
  }
  MaterialExpressionMax_6 {
   class=MaterialExpressionMax
   inputs [
    A {from=MaterialExpressionTextureSampleParameter2D_12, output=R}
    B {from=MaterialExpressionTextureSampleParameter2D_12, output=R}
   ]
  }
  MaterialExpressionMax_7 {
   class=MaterialExpressionMax
   inputs [
    A {from=MaterialExpressionMax_6}
    B {from=MaterialExpressionTextureSampleParameter2D_12, output=B}
   ]
  }
  MaterialExpressionMax_8 {
   class=MaterialExpressionMax
   inputs [
    A {from=MaterialExpressionSaturate_2}
    B {from=MaterialExpressionSaturate_4}
   ]
  }
  MaterialExpressionMultiply_24 {
   class=MaterialExpressionMultiply
   inputs [
    A {from=MaterialExpressionMultiply_47}
    B {from=MaterialExpressionLinearInterpolate_12}
   ]
  }
  MaterialExpressionMultiply_30 {
   class=MaterialExpressionMultiply
   inputs [
    A {from=MaterialExpressionVertexNormalWS_2}
    B {from=MaterialExpressionMultiply_55}
   ]
  }
  MaterialExpressionMultiply_31 {
   class=MaterialExpressionMultiply
   inputs [
    A {from=MaterialExpressionTextureSampleParameter2D_8, output=R}
    B {from=MaterialExpressionVectorParameter_0, output=RGB}
   ]
  }
  MaterialExpressionMultiply_32 {
   class=MaterialExpressionMultiply
   inputs [
    A {from=MaterialExpressionTextureSampleParameter2D_8, output=B}
    B {from=MaterialExpressionVectorParameter_2, output=RGB}
   ]
  }
  MaterialExpressionMultiply_33 {
   class=MaterialExpressionMultiply
   inputs [
    A {from=MaterialExpressionTextureSampleParameter2D_9, output=G}
    B {from=MaterialExpressionVectorParameter_5, output=RGB}
   ]
  }
  MaterialExpressionMultiply_34 {
   class=MaterialExpressionMultiply
   inputs [
    A {from=MaterialExpressionLinearInterpolate_16}
    B {from=""}
   ]
   properties {constB=0.2}
  }
  MaterialExpressionMultiply_35 {
   class=MaterialExpressionMultiply
   inputs [
    A {from=MaterialExpressionLinearInterpolate_18}
    B {from=MaterialExpressionLinearInterpolate_19}
   ]
  }
  MaterialExpressionMultiply_36 {
   class=MaterialExpressionMultiply
   inputs [
    A {from=MaterialExpressionComponentMask_0}
    B {from=MaterialExpressionConstant_16}
   ]
  }
  MaterialExpressionMultiply_37 {
   class=MaterialExpressionMultiply
   inputs [
    A {from=MaterialExpressionDotProduct_1}
    B {from=MaterialExpressionScalarParameter_0}
   ]
   properties {constB=5}
  }
  MaterialExpressionMultiply_38 {
   class=MaterialExpressionMultiply
   inputs [
    A {from=MaterialExpressionNamedRerouteDeclaration_10}
    B {from=MaterialExpressionOneMinus_2}
   ]
  }
  MaterialExpressionMultiply_39 {
   class=MaterialExpressionMultiply
   inputs [
    A {from=MaterialExpressionComponentMask_0}
    B {from=MaterialExpressionConstant_14}
   ]
  }
  MaterialExpressionMultiply_40 {
   class=MaterialExpressionMultiply
   inputs [
    A {from=MaterialExpressionMaterialFunctionCall_6, output=SheerWill_FresnelMask}
    B {from=MaterialExpressionLinearInterpolate_20}
   ]
  }
  MaterialExpressionMultiply_41 {
   class=MaterialExpressionMultiply
   inputs [
    A {from=MaterialExpressionNamedRerouteUsage_12}
    B {from=""}
   ]
   properties {constB=1.6}
  }
  MaterialExpressionMultiply_42 {
   class=MaterialExpressionMultiply
   inputs [
    A {from=MaterialExpressionMultiply_31}
    B {from=MaterialExpressionLinearInterpolate_22}
   ]
  }
  MaterialExpressionMultiply_43 {
   class=MaterialExpressionMultiply
   inputs [
    A {from=MaterialExpressionNamedRerouteUsage_13}
    B {from=""}
   ]
   properties {constB=1.2}
  }
  MaterialExpressionMultiply_44 {
   class=MaterialExpressionMultiply
   inputs [
    A {from=MaterialExpressionMultiply_32}
    B {from=MaterialExpressionLinearInterpolate_23}
   ]
  }
  MaterialExpressionMultiply_45 {
   class=MaterialExpressionMultiply
   inputs [
    A {from=MaterialExpressionNamedRerouteUsage_14}
    B {from=""}
   ]
  }
  MaterialExpressionMultiply_46 {
   class=MaterialExpressionMultiply
   inputs [
    A {from=MaterialExpressionMultiply_33}
    B {from=MaterialExpressionLinearInterpolate_24}
   ]
  }
  MaterialExpressionMultiply_47 {
   class=MaterialExpressionMultiply
   inputs [
    A {from=MaterialExpressionAdd_11}
    B {from=MaterialExpressionNamedRerouteDeclaration_14}
   ]
  }
  MaterialExpressionMultiply_48 {
   class=MaterialExpressionMultiply
   inputs [
    A {from=MaterialExpressionTextureSampleParameter2D_13, output=R}
    B {from=MaterialExpressionConstant_36}
   ]
   properties {constB=0.006}
  }
  MaterialExpressionMultiply_49 {
   class=MaterialExpressionMultiply
   inputs [
    A {from=MaterialExpressionConstant4Vector_9, output=RGB}
    B {from=MaterialExpressionTextureSampleParameter2D_10, output=R}
   ]
  }
  MaterialExpressionMultiply_50 {
   class=MaterialExpressionMultiply
   inputs [
    A {from=MaterialExpressionLinearInterpolate_26}
    B {from=MaterialExpressionMultiply_52}
   ]
  }
  MaterialExpressionMultiply_51 {
   class=MaterialExpressionMultiply
   inputs [
    A {from=MaterialExpressionLinearInterpolate_25}
    B {from=MaterialExpressionOneMinus_3}
   ]
  }
  MaterialExpressionMultiply_52 {
   class=MaterialExpressionMultiply
   inputs [
    A {from=MaterialExpressionMultiply_51}
    B {from=MaterialExpressionConstant_32}
   ]
  }
  MaterialExpressionMultiply_53 {
   class=MaterialExpressionMultiply
   inputs [
    A {from=MaterialExpressionAdd_13}
    B {from=MaterialExpressionTextureSampleParameter2D_11, output=R}
   ]
  }
  MaterialExpressionMultiply_54 {
   class=MaterialExpressionMultiply
   inputs [
    A {from=MaterialExpressionTextureCoordinate_5}
    B {from=MaterialExpressionConstant_33}
   ]
  }
  MaterialExpressionMultiply_55 {
   class=MaterialExpressionMultiply
   inputs [
    A {from=MaterialExpressionMultiply_38}
    B {from=MaterialExpressionLinearInterpolate_28}
   ]
  }
  MaterialExpressionMultiply_56 {
   class=MaterialExpressionMultiply
   inputs [
    A {from=MaterialExpressionTextureSampleParameter2D_15, output=R}
    B {from=""}
   ]
   properties {constB=2}
  }
  MaterialExpressionMultiply_58 {
   class=MaterialExpressionMultiply
   inputs [
    A {from=MaterialExpressionTextureSample_24, output=B}
    B {from=""}
   ]
   properties {constB=3}
  }
  MaterialExpressionNamedRerouteDeclaration_10 {
   class=MaterialExpressionNamedRerouteDeclaration
   inputs [None {from=MaterialExpressionMultiply_35}]
   properties {name=Sheerwill_WPO, nodeColor=<rgba 0 1 0.9297 1>}
  }
  MaterialExpressionNamedRerouteDeclaration_11 {
   class=MaterialExpressionNamedRerouteDeclaration
   inputs [None {from=MaterialExpressionAdd_12}]
   properties {name=OutEmissive, nodeColor=<rgba 0.6641 0 1 1>}
  }
  MaterialExpressionNamedRerouteDeclaration_12 {
   class=MaterialExpressionNamedRerouteDeclaration
   inputs [None {from=MaterialExpressionMaterialFunctionCall_6, output=ProjectionForntUV}]
   properties {name=ProjectionForntUV, nodeColor=<rgba 0 0.0859 1 1>}
  }
  MaterialExpressionNamedRerouteDeclaration_13 {
   class=MaterialExpressionNamedRerouteDeclaration
   inputs [None {from=MaterialExpressionMultiply_24}]
   properties {name=BaseColor_line_1, nodeColor=<rgba 0 1 0.0859 1>}
  }
  MaterialExpressionNamedRerouteDeclaration_14 {
   class=MaterialExpressionNamedRerouteDeclaration
   inputs [None {from=MaterialExpressionPower_4}]
   properties {name=SheerWill_AO, nodeColor=<rgba 0 0.9297 1 1>}
  }
  MaterialExpressionNamedRerouteDeclaration_15 {
   class=MaterialExpressionNamedRerouteDeclaration
   inputs [None {from=MaterialExpressionMultiply_30}]
   properties {name=WPO, nodeColor=<rgba 0.2656 1 0 1>}
  }
  MaterialExpressionNamedRerouteDeclaration_16 {
   class=MaterialExpressionNamedRerouteDeclaration
   inputs [None {from=MaterialExpressionMultiply_50}]
   properties {name=EmissiveChest, nodeColor=<rgba 1 0.9844 0 1>}
  }
  MaterialExpressionNamedRerouteDeclaration_17 {
   class=MaterialExpressionNamedRerouteDeclaration
   inputs [None {from=MaterialExpressionMultiply_40}]
   properties {name=FresnelEmisssive, nodeColor=<rgba 0.6641 0 1 1>}
  }
  MaterialExpressionNamedRerouteDeclaration_8 {
   class=MaterialExpressionNamedRerouteDeclaration
   inputs [None {from=MaterialExpressionSaturate_1}]
   properties {name=FacingMask, nodeColor=<rgba 0 1 0.6719 1>}
  }
  MaterialExpressionNamedRerouteDeclaration_9 {
   class=MaterialExpressionNamedRerouteDeclaration
   inputs [None {from=MaterialExpressionPower_2}]
   properties {name=EdgeNoise2, nodeColor=<rgba 1 0.1172 0 1>}
  }
  MaterialExpressionNamedRerouteUsage_10 {class=MaterialExpressionNamedRerouteUsage}
  MaterialExpressionNamedRerouteUsage_11 {class=MaterialExpressionNamedRerouteUsage}
  MaterialExpressionNamedRerouteUsage_12 {class=MaterialExpressionNamedRerouteUsage}
  MaterialExpressionNamedRerouteUsage_13 {class=MaterialExpressionNamedRerouteUsage}
  MaterialExpressionNamedRerouteUsage_14 {class=MaterialExpressionNamedRerouteUsage}
  MaterialExpressionNamedRerouteUsage_15 {class=MaterialExpressionNamedRerouteUsage}
  MaterialExpressionNamedRerouteUsage_16 {class=MaterialExpressionNamedRerouteUsage}
  MaterialExpressionNamedRerouteUsage_17 {class=MaterialExpressionNamedRerouteUsage}
  MaterialExpressionNamedRerouteUsage_18 {class=MaterialExpressionNamedRerouteUsage}
  MaterialExpressionNamedRerouteUsage_19 {class=MaterialExpressionNamedRerouteUsage}
  MaterialExpressionNamedRerouteUsage_2 {class=MaterialExpressionNamedRerouteUsage}
  MaterialExpressionOneMinus_2 {class=MaterialExpressionOneMinus, inputs [None {from=MaterialExpressionNamedRerouteDeclaration_8}]}
  MaterialExpressionOneMinus_3 {class=MaterialExpressionOneMinus, inputs [None {from=MaterialExpressionFresnel_4}]}
  MaterialExpressionPanner_0 {
   class=MaterialExpressionPanner
   inputs [
    Coordinate {from=MaterialExpressionMultiply_36}
    Time {from=""}
    Speed {from=""}
   ]
   properties {speedX=0.01, speedY=0.01}
  }
  MaterialExpressionPanner_10 {
   class=MaterialExpressionPanner
   inputs [
    Coordinate {from=MaterialExpressionMultiply_41}
    Time {from=""}
    Speed {from=""}
   ]
   properties {speedX=0.04, speedY=0.04}
  }
  MaterialExpressionPanner_11 {
   class=MaterialExpressionPanner
   inputs [
    Coordinate {from=MaterialExpressionMultiply_45}
    Time {from=""}
    Speed {from=""}
   ]
   properties {speedX=0.02, speedY=0.02}
  }
  MaterialExpressionPanner_12 {
   class=MaterialExpressionPanner
   inputs [
    Coordinate {from=MaterialExpressionMultiply_43}
    Time {from=""}
    Speed {from=""}
   ]
   properties {speedX=0.01, speedY=0.01}
  }
  MaterialExpressionPanner_13 {
   class=MaterialExpressionPanner
   inputs [
    Coordinate {from=MaterialExpressionMultiply_39}
    Time {from=""}
    Speed {from=""}
   ]
   properties {speedY=0.1}
  }
  MaterialExpressionPanner_9 {
   class=MaterialExpressionPanner
   inputs [
    Coordinate {from=MaterialExpressionNamedRerouteUsage_11}
    Time {from=""}
    Speed {from=""}
   ]
   properties {speedX=0.02, speedY=0.04}
  }
  MaterialExpressionPower_2 {
   class=MaterialExpressionPower
   inputs [
    Base {from=MaterialExpressionTextureSample_14, output=R}
    Exp {from=MaterialExpressionConstant_12}
   ]
  }
  MaterialExpressionPower_3 {
   class=MaterialExpressionPower
   inputs [
    Base {from=MaterialExpressionVertexInterpolator_1, output=PS}
    Exp {from=""}
   ]
  }
  MaterialExpressionPower_4 {
   class=MaterialExpressionPower
   inputs [
    Base {from=MaterialExpressionTextureSample_20, output=R}
    Exp {from=""}
   ]
   properties {constExponent=2.2}
  }
  MaterialExpressionPower_8 {
   class=MaterialExpressionPower
   inputs [
    Base {from=MaterialExpressionTextureSample_21, output=R}
    Exp {from=""}
   ]
   properties {constExponent=1.5}
  }
  MaterialExpressionSaturate_1 {class=MaterialExpressionSaturate, inputs [None {from=MaterialExpressionMultiply_37}]}
  MaterialExpressionSaturate_2 {class=MaterialExpressionSaturate, inputs [None {from=MaterialExpressionMultiply_56}]}
  MaterialExpressionSaturate_4 {class=MaterialExpressionSaturate, inputs [None {from=MaterialExpressionMultiply_58}]}
  MaterialExpressionScalarParameter_0 {class=MaterialExpressionScalarParameter, properties {defaultValue=0.02, parameterName=FacingMaskStrenth}}
  MaterialExpressionScalarParameter_4 {class=MaterialExpressionScalarParameter, properties {group=Pickaxe_BurnLeer, parameterName=WPOIntensityMin}}
  MaterialExpressionScalarParameter_5 {class=MaterialExpressionScalarParameter, properties {defaultValue=0.003, parameterName=Subsurface Mask Value}}
  MaterialExpressionScalarParameter_6 {class=MaterialExpressionScalarParameter, properties {defaultValue=1, parameterName=WPOIntensityMax}}
  MaterialExpressionSetMaterialAttributes_1 {
   class=MaterialExpressionSetMaterialAttributes
   inputs [
    MaterialAttributes {from=MaterialExpressionMaterialFunctionCall_3, output=Result}
    Emissive Color {from=MaterialExpressionNamedRerouteUsage_10, output=""}
    Base Color {from=MaterialExpressionConstant_1, output=""}
    Roughness {from=MaterialExpressionConstant_7, output=""}
    Specular {from=MaterialExpressionConstant_10, output=""}
    Normal {from=MaterialExpressionTextureSampleParameter2D_14, output=RGB}
    Fuzz Color {from=MaterialExpressionConvert_1, output=""}
    World Position Offset {from=MaterialExpressionNamedRerouteUsage_17, output=""}
   ]
   properties {attributeSetTypes [
     #1
     #2
     #3
     #4
     #5
     #6
     #7
    ]}
  }
  MaterialExpressionSmoothStep_1 {
   class=MaterialExpressionSmoothStep
   inputs [
    Min {from="", output=""}
    Max {from="", output=""}
    Value {from=MaterialExpressionComponentMask_2, output=""}
   ]
   properties {constMax=0.3, constMin=0.3}
  }
  MaterialExpressionSubtract_1 {
   class=MaterialExpressionSubtract
   inputs [
    A {from=MaterialExpressionCameraPositionWS_1}
    B {from=MaterialExpressionWorldPosition_1, output=XYZ}
   ]
  }
  MaterialExpressionTextureCoordinate_3 {class=MaterialExpressionTextureCoordinate}
  MaterialExpressionTextureCoordinate_4 {class=MaterialExpressionTextureCoordinate}
  MaterialExpressionTextureCoordinate_5 {class=MaterialExpressionTextureCoordinate}
  MaterialExpressionTextureSampleParameter2D_10 {
   class=MaterialExpressionTextureSampleParameter2D
   inputs [
    UVs {from="", output=""}
    Apply View MipBias {from="", output=""}
   ]
   properties {
    group=Sheerwill_Phase
    parameterName=SheerWill_FX2
    samplerSource=SSM_Wrap_WorldGroupSettings
    samplerType=SAMPLERTYPE_Masks
    texture=/BRCosmetics/Characters/Player/Male/Medium/Bodies/M_MED_SheerWill_Phase/Textures/T_SheerWill_Phase_Body_FX1
   }
  }
  MaterialExpressionTextureSampleParameter2D_11 {
   class=MaterialExpressionTextureSampleParameter2D
   inputs [
    UVs {from="", output=""}
    Apply View MipBias {from="", output=""}
   ]
   properties {
    group=Sheerwill_Phase
    parameterName=SheerWill_E
    texture=/BRCosmetics/Characters/Player/Male/Medium/Bodies/M_MED_SheerWill_Phase/Textures/T_SheerWill_Phase_Body_E
   }
  }
  MaterialExpressionTextureSampleParameter2D_12 {
   class=MaterialExpressionTextureSampleParameter2D
   inputs [
    UVs {from="", output=""}
    Apply View MipBias {from="", output=""}
   ]
   properties {
    group=Sheerwill_Phase
    parameterName=SheerWill_FX
    samplerType=SAMPLERTYPE_Masks
    texture=/BRCosmetics/Characters/Player/Male/Medium/Bodies/M_MED_SheerWill_Phase/Textures/T_SheerWill_Phase_Body_FX
   }
  }
  MaterialExpressionTextureSampleParameter2D_13 {
   class=MaterialExpressionTextureSampleParameter2D
   inputs [
    UVs {from="", output=""}
    Apply View MipBias {from="", output=""}
   ]
   properties {
    group=Sheerwill_Phase
    parameterName=ParamSheerWill_N
    samplerType=SAMPLERTYPE_Normal
    texture=/BRCosmetics/Characters/Player/Male/Medium/Bodies/M_MED_SheerWill_Phase/Textures/T_SheerWill_Phase_Body_N
   }
  }
  MaterialExpressionTextureSampleParameter2D_14 {
   class=MaterialExpressionTextureSampleParameter2D
   inputs [
    UVs {from="", output=""}
    Apply View MipBias {from="", output=""}
   ]
   properties {
    group=Sheerwill_Phase
    parameterName=ParamSheerWill_N
    samplerType=SAMPLERTYPE_Normal
    texture=&T_SheerWill_Phase_Body_N
   }
  }
  MaterialExpressionTextureSampleParameter2D_15 {
   class=MaterialExpressionTextureSampleParameter2D
   inputs [
    UVs {from="", output=""}
    Apply View MipBias {from="", output=""}
   ]
   properties {
    group=Sheerwill_Phase
    parameterName=SheerWill_E
    texture=&T_SheerWill_Phase_Body_E
   }
  }
  MaterialExpressionTextureSampleParameter2D_8 {
   class=MaterialExpressionTextureSampleParameter2D
   inputs [
    UVs {from="", output=""}
    Apply View MipBias {from="", output=""}
   ]
   properties {
    group=Sheerwill_Phase
    parameterName=SheerWill_FX
    samplerType=SAMPLERTYPE_Masks
    texture=&T_SheerWill_Phase_Body_FX
   }
  }
  MaterialExpressionTextureSampleParameter2D_9 {
   class=MaterialExpressionTextureSampleParameter2D
   inputs [
    UVs {from=MaterialExpressionAdd_10, output=""}
    Apply View MipBias {from="", output=""}
   ]
   properties {
    group=Sheerwill_Phase
    parameterName=SheerWill_FX
    samplerType=SAMPLERTYPE_Masks
    texture=&T_SheerWill_Phase_Body_FX
   }
  }
  MaterialExpressionTextureSample_10 {
   class=MaterialExpressionTextureSample
   inputs [
    UVs {from=""}
    Tex {from=""}
    Apply View MipBias {from=""}
   ]
   properties {samplerType=SAMPLERTYPE_Masks, texture=/BRCosmetics/Characters/Player/Male/Medium/Bodies/M_MED_SheerWill_Phase/Textures/T_Sheerwill_Facemask}
  }
  MaterialExpressionTextureSample_11 {
   class=MaterialExpressionTextureSample
   inputs [
    UVs {from=MaterialExpressionPanner_9}
    Tex {from=""}
    Apply View MipBias {from=""}
   ]
   properties {
    samplerSource=SSM_Wrap_WorldGroupSettings
    samplerType=SAMPLERTYPE_Grayscale
    texture=/Game/Global/Textures/Default/Noise/LowResBlurredNoiseNormalized
   }
  }
  MaterialExpressionTextureSample_13 {
   class=MaterialExpressionTextureSample
   inputs [
    UVs {from=MaterialExpressionPanner_0}
    Tex {from=""}
    Apply View MipBias {from=""}
   ]
   properties {texture=/Game/Global/Textures/Default/Noise/T_CellNoise_01}
  }
  MaterialExpressionTextureSample_14 {
   class=MaterialExpressionTextureSample
   inputs [
    UVs {from=MaterialExpressionPanner_13}
    Tex {from=""}
    Apply View MipBias {from=""}
   ]
   properties {texture=/Game/Global/Textures/Default/Noise/T_SoftCells_Negative}
  }
  MaterialExpressionTextureSample_16 {
   class=MaterialExpressionTextureSample
   inputs [
    UVs {from=MaterialExpressionPanner_10}
    Tex {from=""}
    Apply View MipBias {from=""}
   ]
   properties {samplerType=SAMPLERTYPE_Grayscale, texture=&LowResBlurredNoiseNormalized}
  }
  MaterialExpressionTextureSample_17 {
   class=MaterialExpressionTextureSample
   inputs [
    UVs {from=""}
    Tex {from=""}
    Apply View MipBias {from=""}
   ]
   properties {
    samplerSource=SSM_Wrap_WorldGroupSettings
    samplerType=SAMPLERTYPE_Grayscale
    texture=&LowResBlurredNoiseNormalized
   }
  }
  MaterialExpressionTextureSample_18 {
   class=MaterialExpressionTextureSample
   inputs [
    UVs {from=MaterialExpressionPanner_12}
    Tex {from=""}
    Apply View MipBias {from=""}
   ]
   properties {samplerType=SAMPLERTYPE_Grayscale, texture=&LowResBlurredNoiseNormalized}
  }
  MaterialExpressionTextureSample_19 {
   class=MaterialExpressionTextureSample
   inputs [
    UVs {from=MaterialExpressionPanner_11}
    Tex {from=""}
    Apply View MipBias {from=""}
   ]
   properties {samplerType=SAMPLERTYPE_Grayscale, texture=&LowResBlurredNoiseNormalized}
  }
  MaterialExpressionTextureSample_20 {
   class=MaterialExpressionTextureSample
   inputs [
    UVs {from=""}
    Tex {from=""}
    Apply View MipBias {from=""}
   ]
   properties {
    samplerType=SAMPLERTYPE_Masks
    texture=/BRCosmetics/Characters/Player/Male/Medium/Bodies/M_MED_SheerWill_Phase/Textures/SheerWill_Phase_Body_AO
   }
  }
  MaterialExpressionTextureSample_21 {
   class=MaterialExpressionTextureSample
   inputs [
    UVs {from=MaterialExpressionMultiply_54}
    Tex {from=""}
    Apply View MipBias {from=""}
   ]
   properties {samplerType=SAMPLERTYPE_Grayscale, texture=&LowResBlurredNoiseNormalized}
  }
  MaterialExpressionTextureSample_24 {
   class=MaterialExpressionTextureSample
   inputs [
    UVs {from=""}
    Tex {from=""}
    Apply View MipBias {from=""}
   ]
   properties {samplerType=SAMPLERTYPE_Masks, texture=/BRCosmetics/Characters/Player/Male/Medium/Bodies/M_MED_SheerWill_Phase/Textures/T_Sheerwill_Handmask}
  }
  MaterialExpressionVectorParameter_0 {class=MaterialExpressionVectorParameter, properties {group=Sheerwill_Phase, parameterName=Color_R, defaultValue=<rgba 0.5 0 0.1078 1>}}
  MaterialExpressionVectorParameter_2 {class=MaterialExpressionVectorParameter, properties {parameterName=Color_B, defaultValue=<rgba 0.441 0 0.45 1>}}
  MaterialExpressionVectorParameter_5 {class=MaterialExpressionVectorParameter, properties {parameterName=Color_G, defaultValue=<rgba 0.65 0.026 0.3875 1>}}
  MaterialExpressionVectorParameter_6 {
   class=MaterialExpressionVectorParameter
   properties {group=Sheerwill_Phase, parameterName=SheerWill_EmissiveGlowA, defaultValue=<rgba 1 0 0.1294 1>}
  }
  MaterialExpressionVectorParameter_7 {
   class=MaterialExpressionVectorParameter
   properties {group=Sheerwill_Phase, parameterName=SheerWill_EmissiveGlowC, defaultValue=<rgba 3 0.3041 0 1>}
  }
  MaterialExpressionVectorParameter_8 {
   class=MaterialExpressionVectorParameter
   properties {group=Sheerwill_Phase, parameterName=SheerWill_EmissiveGlowB, defaultValue=<rgba 3 0 0.4709 1>}
  }
  MaterialExpressionVertexInterpolator_1 {class=MaterialExpressionVertexInterpolator, inputs [VS {from=MaterialExpressionNamedRerouteUsage_2, output=""}], properties {}}
  MaterialExpressionVertexNormalWS_0 {class=MaterialExpressionVertexNormalWS, inputs [], properties {}}
  MaterialExpressionVertexNormalWS_1 {class=MaterialExpressionVertexNormalWS, inputs [], properties {}}
  MaterialExpressionVertexNormalWS_2 {class=MaterialExpressionVertexNormalWS, inputs [], properties {}}
  MaterialExpressionVertexNormalWS_3 {class=MaterialExpressionVertexNormalWS, inputs [], properties {}}
  MaterialExpressionWorldPosition_1 {class=MaterialExpressionWorldPosition, inputs [], properties {}}
 }
 properties {
  bAutomaticallySetUsageInEditor=false
  bUseMaterialAttributes=true
  bUsedWithMeshDeformer=true
  bUsedWithMorphTargets=true
  bUsedWithSkeletalMesh=true
  pixelDepthOffsetMode=PDOM_Legacy
  shadingModel=MSM_Cloth
  usedShadingModels=Cloth
 }
}
